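# Translucent UI material for SnowKnight cosmetic item cards, blending a base color toward Chroma, Galaxy or Ice looks through scalar parameters, then darkening with a vignette.
# Chroma maps a warped mask through a gradient texture for animated thin-film color.
# Galaxy layers rotating nebula and star textures, zeroing the large-star blend on mobile.
# Ice fakes lighting from a normal map using half-Lambert diffuse and two reflection-vector speculars.
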Material /BRCosmetics/Athena/UI/Frontend/CosmeticItemCard/Materials/SnowKnight/M_UI_SnowKnight
graph {
 properties {blendMode=BLEND_Translucent, materialDomain=MD_UI, pixelDepthOffsetMode=PDOM_Legacy, usedShadingModels=Default Lit}
 nodes [
  n1 = ComponentMask(None=VectorParameter() {group=Chroma, parameterName=Chroma_WarpSpeed, defaultValue=<rgba -0.0944 0.0233 0 1>}.RGB) {g=true, r=true}
  n2 = Time()
  n3 = TextureSample(UVs=Add(A=TextureCoordinate() {coordinateIndex=4}, B=Multiply(A=@n1, B=@n2)) {constB=0.5}) {
   samplerSource=SSM_Wrap_WorldGroupSettings
   samplerType=SAMPLERTYPE_Masks
   texture=/BRCosmetics/Athena/UI/Frontend/CosmeticItemCard/Materials/SnowKnight/Textures/T_SnowKnight_UI_Chroma2
  }
  ComponentMask(None=@n3.RGB) {b=true}
  n4 = LinearInterpolate(
   A=Desaturation(
    None=VectorParameter() {parameterName=Color, defaultValue=<rgba 0.125 0.125 0.125 1>}.RGB
    Fraction=ScalarParameter() {parameterName=DesaturateAmount}
   )
   B=NamedRerouteUsage()
   Alpha=ScalarParameter() {parameterName=UseChroma}
  )
  n5 = NamedRerouteUsage()
  n6 = ScalarParameter() {parameterName=UseGalaxy}
  LinearInterpolate(
   A=LinearInterpolate(A=LinearInterpolate(A=@n4, B=@n5, Alpha=@n6), B=NamedRerouteUsage(), Alpha=ScalarParameter() {parameterName=UseIce})
   B=Constant3Vector() {constant=<rgba 0 0 0 1>}
   Alpha=NamedRerouteUsage()
  )
  n7 = Multiply(A=Time(), B=ScalarParameter() {defaultValue=0.5, parameterName=SpecSpeed})
  MaterialFunctionCall(X=Sine(None=@n7), Y=Cosine(None=@n7), Z=Constant()) {materialFunction=&MakeFloat3}
  MaterialFunctionCall(Corner Radius=ScalarParameter() {defaultValue=0.2, group=CornerRadius, parameterName=CornerRadius}) {materialFunction=&MF_UI_CoreUI_RoundedMask}
  n8 = TextureSample(UVs=TextureCoordinate() {coordinateIndex=4}) {
   samplerSource=SSM_Wrap_WorldGroupSettings
   samplerType=SAMPLERTYPE_Masks
   texture=&T_SnowKnight_UI_Chroma2
  }
  n9 = LinearInterpolate(
   B=MaterialFunctionCall(Input=@n8.G, Target Low=Constant(), Target High=Constant() {r=5}) {materialFunction=&RemapValueRange}.Result
   Alpha=ScalarParameter() {defaultValue=0.5, group=Chroma, parameterName=Chroma_FakeHighlightIntensity}
  ) {constA=1}
  n10 = Constant3Vector() {constant=<rgba 0.2 0.2 0.2 0>}
  n11 = TextureCoordinate() {coordinateIndex=4}
  n12 = Multiply(A=@n3.B, B=ScalarParameter() {defaultValue=0.1, group=Chroma, parameterName=Chroma_WarpIntensity})
  n13 = TextureSample(UVs=NamedRerouteDeclaration(None=Add(A=@n11, B=@n12)) {name=PanningUV, nodeColor=<rgba 1 0.8672 0 1>}) {
   samplerSource=SSM_Clamp_WorldGroupSettings
   samplerType=SAMPLERTYPE_LinearColor
   texture=/BRCosmetics/Athena/UI/Frontend/CosmeticItemCard/Materials/SnowKnight/Textures/T_SnowKnight_UI_Chroma
  }
  n14 = Multiply(
   A=ScalarParameter() {defaultValue=2, group=Chroma, parameterName=Chroma_ThinFilmScale, sortPriority=5}
   B=Power(Base=@n13.R, Exp=ScalarParameter() {defaultValue=0.2, group=Chroma, parameterName=Chroma_ThinFilmWarp, sortPriority=4})
  )
  n15 = Multiply(A=Time(), B=Constant() {r=0.05})
  n16 = LinearInterpolate(
   A=ScalarParameter() {defaultValue=0.92, group=Chroma, parameterName=Chroma_ThinFilmOffset1}
   B=ScalarParameter() {defaultValue=0.94, group=Chroma, parameterName=Chroma_ThinFilmOffset2}
   Alpha=ConstantBiasScale(None=Sine(None=@n15))
  )
  n17 = TextureSample(
   UVs=MaterialFunctionCall(X=Add(A=@n14, B=@n16), Y=Constant() {r=0.25}) {materialFunction=&MakeFloat2}.Result
   Tex=TextureObject() {texture=/Game/Characters/Player/Female/Medium/Bodies/F_MED_SnowKnight_Helm/Materials/AdvanceMaterials/T_SnowKnight_Gradients}
  ) {
   samplerSource=SSM_Wrap_WorldGroupSettings
   texture=&T_SnowKnight_Gradients
  }
  n18 = ScalarParameter() {defaultValue=2, group=Chroma, parameterName=Chroma_ThinFilmIntensity, sortPriority=2}
  n19 = ScalarParameter() {defaultValue=0.5, group=Chroma, parameterName=Chroma_ThinFilmExponent, sortPriority=3}
  n20 = LinearInterpolate(A=@n10, B=Add(A=@n10, B=Multiply(A=@n17.RGB, B=@n18)), Alpha=Saturate(None=Power(Base=@n13.R, Exp=@n19)))
  NamedRerouteDeclaration(None=Desaturation(None=Multiply(A=@n9, B=@n20), Fraction=ScalarParameter() {group=Chroma, parameterName=Chroma_Desaturation})) {name=Chroma, nodeColor=<rgba 0 1 0.1797 1>}
  n21 = TextureCoordinate() {coordinateIndex=4}
  n22 = Multiply(A=ScalarParameter() {defaultValue=0.1, group=Galaxy, parameterName=Galaxy_RotationSpeed}, B=Time())
  n23 = Add(A=Multiply(A=Rotator(Coordinate=@n21, Time=@n22)) {constB=0.5}) {constB=0.25}
  n24 = Multiply(
   A=Desaturation(
    None=TextureSample(UVs=@n23) {texture=/BRCosmetics/Athena/UI/Frontend/CosmeticItemCard/Materials/SnowKnight/Textures/T_SnowKnight_UI_Nebula}.RGB
    Fraction=ScalarParameter() {defaultValue=0.3, group=Galaxy, parameterName=Galaxy_Saturation}
   )
   B=ScalarParameter() {defaultValue=0.7, group=Galaxy, parameterName=Galaxy_Brightness}
  )
  n25 = Desaturation(None=@n24, Fraction=ScalarParameter() {defaultValue=0.3, group=Galaxy, parameterName=Galaxy_StarSaturation})
  n26 = Multiply(A=@n25, B=VectorParameter() {group=Galaxy, parameterName=Galaxy_StarBrightness, defaultValue=<rgba 250 250 250 1>}.RGB)
  n27 = Multiply(A=TextureCoordinate() {coordinateIndex=4}) {constB=0.75}
  n28 = Multiply(A=ScalarParameter() {defaultValue=-0.1, group=Galaxy, parameterName=Galaxy_SmallStarRotationSpeed}, B=Time())
  n29 = TextureObjectParameter() {
   group=Galaxy
   parameterName=Galaxy_StarsTex
   samplerType=SAMPLERTYPE_Masks
   texture=/Game/Characters/Player/Male/Medium/Bodies/M_Med_Soldier_04/Skins/Galaxy/Textures/Celestial_T_Stars
  }
  n30 = ComponentMask(None=TextureSample(UVs=Rotator(Coordinate=@n27, Time=@n28) {centerX=0.3, centerY=0.3}, Tex=@n29) {samplerSource=SSM_Wrap_WorldGroupSettings}.RGB) {b=true}
  n31 = Multiply(A=@n30, B=@n30)
  n32 = VectorParameter() {group=Galaxy, parameterName=Galaxy_StarBrightness_1, defaultValue=<rgba 1050 1050 1050 1>}
  n33 = ScalarParameter() {defaultValue=-0.2, group=Galaxy, parameterName=Galaxy_LargeStarRotationSpeed}
  n34 = Time()
  n35 = Multiply(A=Rotator(Coordinate=TextureCoordinate() {coordinateIndex=4, uTiling=2, vTiling=2}, Time=Multiply(A=@n33, B=@n34))) {constB=0.5}
  n36 = ComponentMask(None=TextureSample(UVs=Add(A=@n35) {constB=0.5}, Tex=@n29) {samplerSource=SSM_Wrap_WorldGroupSettings}.RGB) {r=true}
  n37 = TextureCoordinate() {coordinateIndex=4, uTiling=2, vTiling=2}
  n38 = TextureSample(UVs=Panner(Coordinate=Multiply(A=@n37) {constB=0.2}) {speedX=0.06, speedY=0.122}, Tex=@n29) {
   samplerSource=SSM_Wrap_WorldGroupSettings
   samplerType=SAMPLERTYPE_Masks
   texture=&Celestial_T_Stars
  }
  n39 = Multiply(A=Multiply(A=@n36, B=@n36), B=@n38.G)
  n40 = Constant()
  NamedRerouteDeclaration(None=LinearInterpolate(
    A=LinearInterpolate(A=@n24, B=@n26, Alpha=@n31)
    B=Multiply(A=@n25, B=@n32.RGB)
    Alpha=ShadingPathSwitch(Default=@n39, Deferred=@n39, Mobile=@n40)
   )) {name=Galaxy, nodeColor=<rgba 1 0 0.7031 1>}
  n41 = TextureSampleParameter2D(UVs=TextureCoordinate() {coordinateIndex=4}) {
   group=Ice
   parameterName=Ice Normal Tiling Tex
   samplerType=SAMPLERTYPE_Normal
   texture=/BRCosmetics/Athena/UI/Frontend/CosmeticItemCard/Materials/SnowKnight/Textures/T_SnowKnight_UI_IceNormal
  }
  n42 = Normalize(VectorInput=@n41.RGB)
  n43 = DotProduct(A=@n42, B=Normalize(VectorInput=VectorParameter() {group=Ice, parameterName=Ice_FakeLightDir, defaultValue=<rgba 0 1 0.399 1>}.RGB)) {desc="diffuse = n dot l"}
  n44 = Power(Base=Add(A=Multiply(A=@n43) {constB=0.5}) {constB=0.5}) {desc=half lambert}
  n45 = LinearInterpolate(
   A=VectorParameter() {group=Ice, parameterName=Ice_ColorShaded, defaultValue=<rgba 0.0138 0.1701 0.2656 1>}.RGB
   B=VectorParameter() {group=Ice, parameterName=Ice_Color, defaultValue=<rgba 0.0296 0.5642 0.9583 1>}.RGB
   Alpha=Multiply(A=Saturate(None=@n44), B=ScalarParameter() {defaultValue=0.8, group=Ice, parameterName=Ice_DiffuseIntensity})
  )
  n46 = Normalize(VectorInput=VectorParameter() {group=Ice, parameterName=Ice_FakeCamDir1, defaultValue=<rgba 1 0 0 1>}.RGB)
  n47 = Constant3Vector() {constant=<rgba 0 0 -1 0>}
  n48 = Saturate(None=DotProduct(
    A=MaterialFunctionCall!(Normal=@n42, CameraVector=@n46) {materialFunction=/Engine/Functions/Engine_MaterialFunctions02/Utility/CustomReflectionVector}.Reflection Vector
    B=@n47
   ) {desc="spec = r dot v"})
  n49 = ScalarParameter() {defaultValue=4, group=Ice, parameterName=Ice_SpecularPower1}
  n50 = Multiply(
   A=Multiply(A=Power(Base=@n48, Exp=@n49), B=ScalarParameter() {defaultValue=3, group=Ice, parameterName=Ice_SpecularIntensity1})
   B=VectorParameter() {group=Ice, parameterName=Ice_SpecColor1, defaultValue=<rgba 0.6302 0.8906 1 1>}.RGB
  )
  n51 = VectorParameter() {group=Ice, parameterName=Ice_FakeCamDir2, defaultValue=<rgba 0.1514 1 0.0384 1>}
  n52 = DotProduct(
   A=MaterialFunctionCall(Normal=@n42, CameraVector=Normalize(VectorInput=@n51.RGB)) {materialFunction=&CustomReflectionVector}.Reflection Vector
   B=@n47
  ) {desc="spec = r dot v"}
  n53 = Multiply(
   A=Power(Base=Saturate(None=@n52), Exp=ScalarParameter() {defaultValue=2, group=Ice, parameterName=Ice_SpecularPower2})
   B=ScalarParameter() {defaultValue=0.5, group=Ice, parameterName=Ice_SpecularIntensity2}
  )
  n54 = VectorParameter() {group=Ice, parameterName=Ice_SpecColor2, defaultValue=<rgba 0 0.7043 1 1>}
  n55 = Multiply(A=Add(A=Add(A=@n45, B=@n50), B=Multiply(A=@n53, B=@n54.RGB)), B=ScalarParameter() {defaultValue=1, group=Ice, parameterName=Ice_Brightness})
  n56 = LinearInterpolate(A=@n55, B=VectorParameter() {group=Ice, parameterName=Ice_EdgeTint, defaultValue=<rgba 0 0.5127 0.8958 1>}.RGB)
  n57 = ScalarParameter() {defaultValue=0.2, group=Ice, parameterName=Ice_UseEdgeMap}
  NamedRerouteDeclaration(None=Desaturation(None=LinearInterpolate(A=@n55, B=@n56, Alpha=@n57), Fraction=ScalarParameter() {group=Ice, parameterName=Ice_Desaturation})) {name=Ice, nodeColor=<rgba 0 0.6806 1 1>}
  n58 = TextureCoordinate() {coordinateIndex=4}
  n59 = Constant2Vector() {r=0.5}
  n60 = Power(Base=OneMinus(None=SphereMask(A=@n58, B=@n59) {attenuationRadius=2, hardnessPercent=0})) {constExponent=3}
  n61 = ScalarParameter() {defaultValue=2.75, parameterName=VignetteIntensity}
  NamedRerouteDeclaration(None=Saturate(None=Multiply(A=@n60, B=@n61))) {name=Vignette, nodeColor=<rgba 0.0078 0 1 1>}
  NamedRerouteUsage()
 ]
}
